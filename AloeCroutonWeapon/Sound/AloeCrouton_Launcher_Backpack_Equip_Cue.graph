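# Sound cue for the AloeCrouton launcher backpack equip foley, playing one equip wave.
# An IsLocal branch plays it at full mixer level with no attenuation asset for the local player, otherwise deferring to an IsEnemy branch.
# That branch uses weapon reload attenuation, keeping enemy or unset at full level and lowering non-enemy to 0.3.
# Concurrency allows one instance per owner, stopping the oldest.
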SoundCue /AloeCroutonWeapon/Sound/AloeCrouton_Launcher_Backpack_Equip_Cue
root=Branch_1
nodes [
 {id=Attenuation_6, children [WavePlayer_13], properties {attenuationSettings=&Fort_Weapon_Reload_Att}}
 {id=Attenuation_7, children [WavePlayer_14]}
 {id=Attenuation_8, children [WavePlayer_15], properties {attenuationSettings=&Fort_Weapon_Reload_Att}}
 {id=Branch_0, children [Mixer_8,Mixer_6,Mixer_8], properties {boolParameterName=IsEnemy}}
 {id=Branch_1, children [Mixer_7,Branch_0,Branch_0], properties {boolParameterName=IsLocal}}
 {id=Mixer_6, children [Attenuation_8], properties {inputVolume [0.3]}}
 {id=Mixer_7, children [Attenuation_7], properties {inputVolume [1]}}
 {id=Mixer_8, children [Attenuation_6], properties {inputVolume [1]}}
 {id=WavePlayer_13, properties {soundWaveAssetPtr=&/AloeCrouton_Launcher_Backpack_Equip_01}}
 {id=WavePlayer_14, properties {soundWaveAssetPtr=&AloeCrouton_Launcher_Backpack_Equip_01}}
 {id=WavePlayer_15, properties {soundWaveAssetPtr=&AloeCrouton_Launcher_Backpack_Equip_01}}
]
properties {
 bOverrideConcurrency=true
 maxDistance=2097152
 soundClassObject=&Fort_WeaponFoley
 volumeMultiplier=0.752
 attenuationOverrides {falloffDistance=5000, attenuationShapeExtents {x=1000}}
 concurrencyOverrides {bLimitToOwner=true, maxCount=1, resolutionRule=StopOldest}
}
