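# Sound cue for the unequip foley of the AloeCrouton launcher backpack weapon.
# A single wave player feeds the one unequip recording straight to the output, with no randomization or modulation.
# It uses the weapon foley sound class, reload attenuation settings and reduced volume, with a concurrency override limiting playback to one instance at a time.
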SoundCue /AloeCroutonWeapon/Sound/AloeCrouton_Launcher_Backpack_Unequip_Cue
root=WavePlayer_1
nodes [{id=WavePlayer_1, properties {soundWaveAssetPtr=&/AloeCrouton_Launcher_Backpack_Unequip_01}}]
properties {
 attenuationSettings=&Fort_Weapon_Reload_Att
 bOverrideConcurrency=true
 maxDistance=2250
 soundClassObject=&Fort_WeaponFoley
 volumeMultiplier=0.501
 attenuationOverrides {falloffDistance=5000, attenuationShapeExtents {x=1000}}
 concurrencyOverrides {maxCount=1}
}
