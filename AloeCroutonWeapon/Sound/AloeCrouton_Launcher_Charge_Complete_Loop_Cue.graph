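# Sound cue for the AloeCrouton launcher's charge-complete sound, pairing a one-shot start with a sustained loop.
# Branches on IsLocal, then IsEnemy: the local path adds a first-person accent loop, and non-enemy remote playback is mixed at half the enemy level.
# Envelopers fade each loop in while dropping its pitch to normal.
# Remote paths use separate attenuation assets for start and loop, and the cue overrides concurrency.
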SoundCue /AloeCroutonWeapon/Sound/AloeCrouton_Launcher_Charge_Complete_Loop_Cue
root=Branch_0
nodes [
 {id=Attenuation_0, children [Modulator_0], properties {attenuationSettings=&/AloeCrouton_Launcher_Fire_Charge_Att}}
 {
  id=Attenuation_1
  children [Enveloper_0]
  properties {attenuationSettings=&/AloeCrouton_Launcher_Fire_Charge_Loop_Att}
 }
 {id=Attenuation_2, children [Modulator_5]}
 {id=Attenuation_3, children [Enveloper_1]}
 {id=Attenuation_4, children [Modulator_6], properties {attenuationSettings=&AloeCrouton_Launcher_Fire_Charge_Att}}
 {
  id=Attenuation_5
  children [Enveloper_2]
  properties {attenuationSettings=&AloeCrouton_Launcher_Fire_Charge_Loop_Att}
 }
 {id=Branch_0, children [Mixer_4,Branch_1,Branch_1], properties {boolParameterName=IsLocal}}
 {id=Branch_1, children [Mixer_5,Mixer_3,Mixer_5], properties {boolParameterName=IsEnemy}}
 {
  id=Enveloper_0
  children [WavePlayer_0]
  properties {pitchCurve {editorCurveData {keys [<at 0 1.5>, {time=1}]}}, volumeCurve {editorCurveData {keys [<at 0 0>,<at 0.5 1>]}}}
 }
 {
  id=Enveloper_1
  children [Mixer_7]
  properties {pitchCurve {editorCurveData {keys [<at 0 1.5>, {time=1}]}}, volumeCurve {editorCurveData {keys [<at 0 0>,<at 0.5 1>]}}}
 }
 {
  id=Enveloper_2
  children [WavePlayer_4]
  properties {pitchCurve {editorCurveData {keys [<at 0 1.5>, {time=1}]}}, volumeCurve {editorCurveData {keys [<at 0 0>,<at 0.5 1>]}}}
 }
 {id=Mixer_0, children [Attenuation_2, Attenuation_3], properties {inputVolume [0.7,1]}}
 {id=Mixer_1, children [Attenuation_4, Attenuation_5], properties {inputVolume [0.7,1]}}
 {id=Mixer_2, children [Attenuation_0, Attenuation_1], properties {inputVolume [0.7,1]}}
 {id=Mixer_3, children [Mixer_1], properties {inputVolume [0.5]}}
 {id=Mixer_4, children [Mixer_0], properties {inputVolume [1]}}
 {id=Mixer_5, children [Mixer_2], properties {inputVolume [1]}}
 {id=Mixer_7, children [WavePlayer_2,WavePlayer_6], properties {inputVolume [1,1]}}
 {id=Modulator_0, children [WavePlayer_1], properties {volumeMax=1, volumeMin=1}}
 {id=Modulator_5, children [WavePlayer_3], properties {volumeMax=1, volumeMin=1}}
 {id=Modulator_6, children [WavePlayer_5], properties {volumeMax=1, volumeMin=1}}
 {id=WavePlayer_0, properties {bLooping=true, soundWaveAssetPtr=&/AloeCrouton_Launcher_Charge_Complete_Loop_01}}
 {id=WavePlayer_1, properties {soundWaveAssetPtr=&/AloeCrouton_Launcher_Charge_Complete_Loop_Start_01}}
 {id=WavePlayer_2, properties {bLooping=true, soundWaveAssetPtr=&AloeCrouton_Launcher_Charge_Complete_Loop_01}}
 {id=WavePlayer_3, properties {soundWaveAssetPtr=&AloeCrouton_Launcher_Charge_Complete_Loop_Start_01}}
 {id=WavePlayer_4, properties {bLooping=true, soundWaveAssetPtr=&AloeCrouton_Launcher_Charge_Complete_Loop_01}}
 {id=WavePlayer_5, properties {soundWaveAssetPtr=&AloeCrouton_Launcher_Charge_Complete_Loop_Start_01}}
 {id=WavePlayer_6, properties {bLooping=true, soundWaveAssetPtr=&/AloeCrouton_Launcher_Charge_Complete_1P_Accent_Loop}}
]
properties {
 bOverrideConcurrency=true
 maxDistance=2097152
 soundClassObject=&Fort_WeaponRanged_Fire_3P_PSM
 volumeMultiplier=0.65
 attenuationOverrides {distanceAlgorithm=NaturalSound, falloffDistance=5000, attenuationShapeExtents {x=1000}}
 concurrencyOverrides {maxCount=3}
 concurrencySet [&Weapon_Gunfire_Local_Conc, &Weapon_Gunfire_Global_Conc]
}
