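# Sound cue for the start of the AloeCrouton launcher's charge, playing one close build-up wave.
# Nested IsLocal and IsEnemy branches select the listener mix: full volume with unassigned attenuation node settings for local, shared charge attenuation for others.
# Non-local enemies, or an unset enemy flag, play louder than non-enemies.
# It uses the third-person ranged-fire sound class and local and global gunfire concurrency.
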SoundCue /AloeCroutonWeapon/Sound/AloeCrouton_Launcher_Charge_Start_Cue
root=Branch_0
nodes [
 {id=Attenuation_0, children [WavePlayer_3], properties {attenuationSettings=&/AloeCrouton_Launcher_Fire_Charge_Att}}
 {id=Attenuation_1, children [WavePlayer_8], properties {attenuationSettings=&AloeCrouton_Launcher_Fire_Charge_Att}}
 {id=Attenuation_2, children [WavePlayer_7]}
 {id=Branch_0, children [Mixer_4,Branch_1,Branch_1], properties {boolParameterName=IsLocal}}
 {id=Branch_1, children [Mixer_5,Mixer_3,Mixer_5], properties {boolParameterName=IsEnemy}}
 {id=Mixer_3, children [Attenuation_1], properties {inputVolume [0.5]}}
 {id=Mixer_4, children [Attenuation_2], properties {inputVolume [1]}}
 {id=Mixer_5, children [Attenuation_0], properties {inputVolume [0.8]}}
 {id=WavePlayer_3, properties {soundWaveAssetPtr=&/AloeCrouton_Launcher_Fire_Close_Build_Up_01}}
 {id=WavePlayer_7, properties {soundWaveAssetPtr=&AloeCrouton_Launcher_Fire_Close_Build_Up_01}}
 {id=WavePlayer_8, properties {soundWaveAssetPtr=&AloeCrouton_Launcher_Fire_Close_Build_Up_01}}
]
properties {
 maxDistance=2097152
 soundClassObject=&Fort_WeaponRanged_Fire_3P_PSM
 attenuationOverrides {distanceAlgorithm=NaturalSound, falloffDistance=5000, attenuationShapeExtents {x=1000}}
 concurrencySet [&Weapon_Gunfire_Local_Conc, &Weapon_Gunfire_Global_Conc]
}
