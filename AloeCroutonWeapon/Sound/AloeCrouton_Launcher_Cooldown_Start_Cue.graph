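# Sound cue named for the start of the AloeCrouton launcher's cooldown, currently empty with no root node or sound nodes, so it has no audio to play.
# It carries only playback settings: reduced volume, the weapon-drop attenuation asset and the loot-collect sound class.
# It references a pickup concurrency set and also holds single-instance, prevent-new concurrency override values.
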SoundCue /AloeCroutonWeapon/Sound/AloeCrouton_Launcher_Cooldown_Start_Cue
root=null
nodes []
properties {
 attenuationSettings=&Fort_Weapon_Drop_Att
 maxDistance=3000
 soundClassObject=&Fort_Loot_Collect
 volumeMultiplier=0.3231
 concurrencyOverrides {maxCount=1, resolutionRule=PreventNew}
 concurrencySet [&PickupMaximum_Oldest]
}
