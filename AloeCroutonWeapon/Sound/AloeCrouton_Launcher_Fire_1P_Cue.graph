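# First-person fire sound cue for the AloeCrouton launcher weapon.
# A single close-range fire wave plays through a Modulator that pins volume at unity, with a boosted cue volume multiplier.
# It uses the first-person ranged-fire sound class, first-person gunfire concurrency, and first-person projectile attenuation with natural-sound falloff overrides.
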SoundCue /AloeCroutonWeapon/Sound/AloeCrouton_Launcher_Fire_1P_Cue
root=Modulator_1
nodes [
 {id=Modulator_1, children [WavePlayer_0], properties {volumeMax=1, volumeMin=1}}
 {id=WavePlayer_0, properties {soundWaveAssetPtr=&/AloeCrouton_Launcher_Fire_Close_01}}
]
properties {
 attenuationSettings=&Fort_Weapon_Projectile_FirstPerson_Athena_Att
 maxDistance=2097152
 soundClassObject=&Fort_WeaponRanged_Fire_1P_PSM
 volumeMultiplier=1.25
 attenuationOverrides {distanceAlgorithm=NaturalSound, falloffDistance=5000, attenuationShapeExtents {x=1000}}
 concurrencySet [&Weapon_Gunfire_FirstPerson_Conc]
}
