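# Sound cue for the AloeCrouton launcher's firing sound, layered for close, medium and distant range.
# A branch on IsEnemy plays the full mix when true or unset, and a duplicate mix at reduced volume when false.
# Each mix blends a pitch-randomized close shot with randomly chosen mid and distant variants, each layer with its own attenuation and sound class.
# Concurrency is overridden to cap simultaneous instances.
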SoundCue /AloeCroutonWeapon/Sound/AloeCrouton_Launcher_Fire_Cue
root=Branch_1
nodes [
 {
  id=Attenuation_0
  children [Modulator_2]
  properties {attenuationSettings=&Fort_Weapon_Projectile_Medium_Mono_Athena_Att}
 }
 {id=Attenuation_1, children [Modulator_5], properties {attenuationSettings=&Fort_Weapon_Projectile_Close_Athena_Att}}
 {
  id=Attenuation_2
  children [Modulator_1]
  properties {attenuationSettings=&Fort_Weapon_Projectile_Distant_Mono_Athena_Att}
 }
 {
  id=Attenuation_3
  children [Modulator_0]
  properties {attenuationSettings=&Fort_Weapon_Projectile_Medium_Mono_Athena_Att}
 }
 {
  id=Attenuation_34
  children [Modulator_3]
  properties {attenuationSettings=&Fort_Weapon_Projectile_Distant_Mono_Athena_Att}
 }
 {id=Attenuation_49, children [Modulator_4], properties {attenuationSettings=&Fort_Weapon_Projectile_Close_Athena_Att}}
 {id=Branch_1, children [Mixer_17,Mixer_1,Mixer_17], properties {boolParameterName=IsEnemy}}
 {id=Mixer_0, children [SoundClass_1,SoundClass_3,SoundClass_2], properties {inputVolume [1.25,1,1]}}
 {id=Mixer_1, children [Mixer_0], properties {inputVolume [0.7]}}
 {id=Mixer_17, children [SoundClass_32, SoundClass_0, SoundClass_35], properties {inputVolume [1.25,1,1]}}
 {id=Modulator_0, children [Random_4], properties {volumeMax=1, volumeMin=1}}
 {id=Modulator_1, children [Random_5], properties {volumeMax=1, volumeMin=1}}
 {id=Modulator_2, children [Random_1], properties {volumeMax=1, volumeMin=1}}
 {id=Modulator_3, children [Random_3], properties {volumeMax=1, volumeMin=1}}
 {id=Modulator_4, children [WavePlayer_0], properties {pitchMax=1.2, pitchMin=0.9, volumeMax=1, volumeMin=1}}
 {id=Modulator_5, children [WavePlayer_12], properties {pitchMax=1.2, pitchMin=0.9, volumeMax=1, volumeMin=1}}
 {id=Random_1, children [WavePlayer_3,WavePlayer_4,WavePlayer_5], properties {weights [1,1,1]}}
 {id=Random_3, children [WavePlayer_9, WavePlayer_10, WavePlayer_11], properties {weights [1,1,1]}}
 {id=Random_4, children [WavePlayer_13, WavePlayer_14, WavePlayer_15], properties {weights [1,1,1]}}
 {id=Random_5, children [WavePlayer_16, WavePlayer_17, WavePlayer_18], properties {weights [1,1,1]}}
 {
  id=SoundClass_0
  children [Attenuation_0]
  properties {soundClassOverride=&Fort_WeaponRanged_Fire_Distant}
 }
 {
  id=SoundClass_1
  children [Attenuation_1]
  properties {soundClassOverride=&Fort_WeaponRanged_Fire_3P_PSM}
 }
 {
  id=SoundClass_2
  children [Attenuation_2]
  properties {soundClassOverride=&Fort_WeaponRanged_Fire_Distant}
 }
 {
  id=SoundClass_3
  children [Attenuation_3]
  properties {soundClassOverride=&Fort_WeaponRanged_Fire_Distant}
 }
 {
  id=SoundClass_32
  children [Attenuation_49]
  properties {soundClassOverride=&Fort_WeaponRanged_Fire_3P_PSM}
 }
 {
  id=SoundClass_35
  children [Attenuation_34]
  properties {soundClassOverride=&Fort_WeaponRanged_Fire_Distant}
 }
 {id=WavePlayer_0, properties {soundWaveAssetPtr=&/AloeCrouton_Launcher_Fire_Close_01}}
 {id=WavePlayer_10, properties {soundWaveAssetPtr=&/AloeCrouton_Launcher_Fire_Distant_02}}
 {id=WavePlayer_11, properties {soundWaveAssetPtr=&/AloeCrouton_Launcher_Fire_Distant_03}}
 {id=WavePlayer_12, properties {soundWaveAssetPtr=&AloeCrouton_Launcher_Fire_Close_01}}
 {id=WavePlayer_13, properties {soundWaveAssetPtr=&/AloeCrouton_Launcher_Fire_Mid_01}}
 {id=WavePlayer_14, properties {soundWaveAssetPtr=&/AloeCrouton_Launcher_Fire_Mid_02}}
 {id=WavePlayer_15, properties {soundWaveAssetPtr=&/AloeCrouton_Launcher_Fire_Mid_03}}
 {id=WavePlayer_16, properties {soundWaveAssetPtr=&/AloeCrouton_Launcher_Fire_Distant_01}}
 {id=WavePlayer_17, properties {soundWaveAssetPtr=&AloeCrouton_Launcher_Fire_Distant_02}}
 {id=WavePlayer_18, properties {soundWaveAssetPtr=&AloeCrouton_Launcher_Fire_Distant_03}}
 {id=WavePlayer_3, properties {soundWaveAssetPtr=&AloeCrouton_Launcher_Fire_Mid_01}}
 {id=WavePlayer_4, properties {soundWaveAssetPtr=&AloeCrouton_Launcher_Fire_Mid_02}}
 {id=WavePlayer_5, properties {soundWaveAssetPtr=&AloeCrouton_Launcher_Fire_Mid_03}}
 {id=WavePlayer_9, properties {soundWaveAssetPtr=&AloeCrouton_Launcher_Fire_Distant_01}}
]
properties {
 bOverrideConcurrency=true
 maxDistance=27500
 volumeMultiplier=0.85
 attenuationOverrides {distanceAlgorithm=NaturalSound, falloffDistance=5000, attenuationShapeExtents {x=1000}}
 concurrencyOverrides {maxCount=4}
}
