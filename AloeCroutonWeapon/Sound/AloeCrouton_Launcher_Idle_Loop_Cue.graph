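# Looping idle sound cue for the AloeCrouton launcher, playing its full-ammo loop in the weapon foley sound class.
# IsEnemy and IsLocal branches pick one of three mixes of the same wave: loudest for enemies, medium for local, quietest for other non-enemies.
# Enemy and non-local paths share the idle loop attenuation asset, while the local path assigns none; an unset IsLocal falls back to the enemy mix.
# Concurrency is overridden to cap simultaneous instances.
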SoundCue /AloeCroutonWeapon/Sound/AloeCrouton_Launcher_Idle_Loop_Cue
root=Branch_2
nodes [
 {id=Attenuation_0, children [WavePlayer_5], properties {attenuationSettings=&/AloeCrouton_Launcher_Idle_Loop_Att}}
 {id=Attenuation_1, children [WavePlayer_6]}
 {id=Attenuation_2, children [WavePlayer_7], properties {attenuationSettings=&AloeCrouton_Launcher_Idle_Loop_Att}}
 {id=Branch_1, children [Mixer_4,Mixer_3,Mixer_5], properties {boolParameterName=IsLocal}}
 {id=Branch_2, children [Mixer_5,Branch_1,Branch_1], properties {boolParameterName=IsEnemy}}
 {id=Mixer_3, children [Attenuation_2], properties {inputVolume [0.3]}}
 {id=Mixer_4, children [Attenuation_1], properties {inputVolume [1]}}
 {id=Mixer_5, children [Attenuation_0], properties {inputVolume [1.25]}}
 {id=WavePlayer_5, properties {bLooping=true, soundWaveAssetPtr=&/AloeCrouton_Launcher_Ammo_Full_Loop}}
 {id=WavePlayer_6, properties {bLooping=true, soundWaveAssetPtr=&AloeCrouton_Launcher_Ammo_Full_Loop}}
 {id=WavePlayer_7, properties {bLooping=true, soundWaveAssetPtr=&AloeCrouton_Launcher_Ammo_Full_Loop}}
]
properties {
 bOverrideConcurrency=true
 maxDistance=2097152
 soundClassObject=&Fort_WeaponFoley
 volumeMultiplier=0.626
 concurrencyOverrides {maxCount=3}
}
