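# Sound cue for the AloeCrouton launcher's low-ammo firing sound.
# A Random node picks one of three equally weighted low-ammo fire waves and feeds a Modulator that varies pitch slightly while holding volume steady.
# It plays through a low-ammo ranged-weapon sound class with close projectile attenuation, and a concurrency override caps simultaneous instances at two.
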SoundCue /AloeCroutonWeapon/Sound/AloeCrouton_Launcher_Low_Ammo_Cue
root=Modulator_0
nodes [
 {id=Modulator_0, children [Random_0], properties {pitchMax=1.2, pitchMin=0.9, volumeMax=1, volumeMin=1}}
 {id=Random_0, children [WavePlayer_1,WavePlayer_2,WavePlayer_3], properties {bShouldExcludeFromBranchCulling=true, weights [1,1,1]}}
 {id=WavePlayer_1, properties {soundWaveAssetPtr=&/AloeCrouton_Launcher_Fire_Low_Ammo_01}}
 {id=WavePlayer_2, properties {soundWaveAssetPtr=&/AloeCrouton_Launcher_Fire_Low_Ammo_02}}
 {id=WavePlayer_3, properties {soundWaveAssetPtr=&/AloeCrouton_Launcher_Fire_Low_Ammo_03}}
]
properties {
 attenuationSettings=&Fort_Weapon_Projectile_Close_Athena_Att
 bOverrideConcurrency=true
 maxDistance=4950
 soundClassObject=/Game/Sounds/FortSoundClassesAndModes/Fort_WeaponRanged_LowAmmo_Degraded_PSM
 volumeMultiplier=1
 concurrencyOverrides {maxCount=2}
}
