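# Sound cue for the end of the AloeCrouton launcher's overheat, limited to one concurrent instance.
# An IsLocal branch plays the wave through the first-person ranged-fire sound class, or passes to an IsEnemy branch for other players.
# Both enemy and non-enemy paths use the third-person class at half mixer volume with the launcher's fire-charge attenuation asset, so they currently sound identical.
# All paths share one wave.
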SoundCue /AloeCroutonWeapon/Sound/AloeCrouton_Launcher_Overheat_End_Cue
root=Branch_2
nodes [
 {id=Attenuation_6, children [WavePlayer_13], properties {attenuationSettings=&/AloeCrouton_Launcher_Fire_Charge_Att}}
 {id=Attenuation_7, children [WavePlayer_14]}
 {id=Attenuation_8, children [WavePlayer_15], properties {attenuationSettings=&AloeCrouton_Launcher_Fire_Charge_Att}}
 {id=Branch_2, children [Mixer_7,Branch_3,Branch_3], properties {boolParameterName=IsLocal}}
 {id=Branch_3, children [Mixer_8,Mixer_6,Mixer_8], properties {boolParameterName=IsEnemy}}
 {id=Mixer_6, children [SoundClass_0], properties {inputVolume [0.5]}}
 {id=Mixer_7, children [SoundClass_2], properties {inputVolume [1]}}
 {id=Mixer_8, children [SoundClass_1], properties {inputVolume [0.5]}}
 {
  id=SoundClass_0
  children [Attenuation_8]
  properties {soundClassOverride=&Fort_WeaponRanged_Fire_3P_PSM}
 }
 {
  id=SoundClass_1
  children [Attenuation_6]
  properties {soundClassOverride=&Fort_WeaponRanged_Fire_3P_PSM}
 }
 {
  id=SoundClass_2
  children [Attenuation_7]
  properties {soundClassOverride=&Fort_WeaponRanged_Fire_1P_PSM}
 }
 {id=WavePlayer_13, properties {soundWaveAssetPtr=&/AloeCrouton_Launcher_Overheat_End}}
 {id=WavePlayer_14, properties {soundWaveAssetPtr=&AloeCrouton_Launcher_Overheat_End}}
 {id=WavePlayer_15, properties {soundWaveAssetPtr=&AloeCrouton_Launcher_Overheat_End}}
]
properties {
 bOverrideConcurrency=true
 maxDistance=2097152
 soundClassObject=&Fort_WeaponFoley
 volumeMultiplier=0.827
 concurrencyOverrides {maxCount=1}
}
