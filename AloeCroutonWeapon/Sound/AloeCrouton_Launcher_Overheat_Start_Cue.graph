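# Sound cue for the overheat start sound of the AloeCrouton launcher, playing one wave with a mix that depends on the listener.
# IsLocal selects a first-person sound class path; otherwise IsEnemy selects a louder enemy path, also used when unset, or a quieter third-person class path for non-enemies.
# Both non-local paths reuse the launcher's fire charge attenuation.
# Overridden concurrency limits simultaneous plays per owner.
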SoundCue /AloeCroutonWeapon/Sound/AloeCrouton_Launcher_Overheat_Start_Cue
root=Branch_2
nodes [
 {id=Attenuation_0, children [WavePlayer_5], properties {attenuationSettings=&/AloeCrouton_Launcher_Fire_Charge_Att}}
 {id=Attenuation_1, children [WavePlayer_6]}
 {id=Attenuation_2, children [WavePlayer_7], properties {attenuationSettings=&AloeCrouton_Launcher_Fire_Charge_Att}}
 {id=Branch_2, children [Mixer_4,Branch_3,Branch_3], properties {boolParameterName=IsLocal}}
 {id=Branch_3, children [Mixer_5,Mixer_3,Mixer_5], properties {boolParameterName=IsEnemy}}
 {id=Mixer_3, children [SoundClass_2], properties {inputVolume [0.3]}}
 {id=Mixer_4, children [SoundClass_0], properties {inputVolume [1]}}
 {id=Mixer_5, children [Attenuation_0], properties {inputVolume [1.1]}}
 {
  id=SoundClass_0
  children [Attenuation_1]
  properties {soundClassOverride=&Fort_WeaponRanged_Fire_1P_PSM}
 }
 {
  id=SoundClass_2
  children [Attenuation_2]
  properties {soundClassOverride=&Fort_WeaponRanged_Fire_3P_PSM}
 }
 {id=WavePlayer_5, properties {soundWaveAssetPtr=&/AloeCrouton_Launcher_Overheat_Start}}
 {id=WavePlayer_6, properties {soundWaveAssetPtr=&AloeCrouton_Launcher_Overheat_Start}}
 {id=WavePlayer_7, properties {soundWaveAssetPtr=&AloeCrouton_Launcher_Overheat_Start}}
]
properties {
 bOverrideConcurrency=true
 maxDistance=2097152
 volumeMultiplier=2
 attenuationOverrides {falloffDistance=5000, attenuationShapeExtents {x=1000}}
 concurrencyOverrides {bLimitToOwner=true, maxCount=2}
}
