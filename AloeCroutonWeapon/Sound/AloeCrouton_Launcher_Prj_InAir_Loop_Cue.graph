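# Sound cue for the in-flight loop of the AloeCrouton launcher projectile.
# A mixer layers one random Egg Launcher start sound, through a modulator, over the weapon's looping wave, each with its own attenuation.
# An enveloper fades the loop in while bending its pitch down, then a smoothed Doppler node shifts pitch with relative motion.
# The cue overrides concurrency to cap simultaneous instances.
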SoundCue /AloeCroutonWeapon/Sound/AloeCrouton_Launcher_Prj_InAir_Loop_Cue
root=Mixer_2
nodes [
 {id=Attenuation_3, children [Doppler_0], properties {attenuationSettings=/Game/Sounds/Fort_Weapons/Launchers/Launcher_Prj_Loop_Att}}
 {id=Attenuation_4, children [Modulator_1], properties {attenuationSettings=/Game/Sounds/Fort_Weapons/Launchers/Launcher_Prj_Loop_Start_Att}}
 {id=Doppler_0, children [Enveloper_0], properties {bUseSmoothing=true, dopplerIntensity=1.55}}
 {
  id=Enveloper_0
  children [WavePlayer_8]
  properties {pitchCurve {editorCurveData {keys [<at 0 1.5>, {time=1}]}}, volumeCurve {editorCurveData {keys [<at 0 0 RCIM_Cubic>,<at 0.75 1>]}}}
 }
 {id=Mixer_2, children [Attenuation_4, Attenuation_3], properties {inputVolume [1.5,1.35]}}
 {id=Modulator_1, children [Random_0], properties {volumeMax=1, volumeMin=1}}
 {id=Random_0, children [WavePlayer_4,WavePlayer_5,WavePlayer_6,WavePlayer_7], properties {weights [1,1,1,1]}}
 {id=WavePlayer_4, properties {soundWaveAssetPtr=/Game/Sounds/Fort_Weapons/Launchers/Egg_Launcher/Egg_Launcher_Prj_InAir_Loop_Start_01}}
 {id=WavePlayer_5, properties {soundWaveAssetPtr=/Game/Sounds/Fort_Weapons/Launchers/Egg_Launcher/Egg_Launcher_Prj_InAir_Loop_Start_02}}
 {id=WavePlayer_6, properties {soundWaveAssetPtr=/Game/Sounds/Fort_Weapons/Launchers/Egg_Launcher/Egg_Launcher_Prj_InAir_Loop_Start_03}}
 {id=WavePlayer_7, properties {soundWaveAssetPtr=/Game/Sounds/Fort_Weapons/Launchers/Egg_Launcher/Egg_Launcher_Prj_InAir_Loop_Start_04}}
 {id=WavePlayer_8, properties {bLooping=true, soundWaveAssetPtr=&/AloeCrouton_Launcher_Prj_InAir_Loop}}
]
properties {
 bOverrideConcurrency=true
 maxDistance=19000
 soundClassObject=&Fort_WeaponRanged_Fire_3P_PSM
 volumeMultiplier=1
 concurrencyOverrides {maxCount=4}
}
