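# Sound cue for the close projectile whip of the AloeCrouton launcher, classed as a ranged-weapon projectile effect.
# A Random node picks one of four equally weighted whip waves and feeds a Modulator at default settings for slight pitch and volume variation.
# A shared whip attenuation asset handles distance behavior, while a cue-level concurrency override caps simultaneous instances.
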SoundCue /AloeCroutonWeapon/Sound/AloeCrouton_Launcher_Prj_Whip_Close_Cue
root=Modulator_0
nodes [
 {id=Modulator_0, children [Random_1]}
 {
  id=Random_1
  children [WavePlayer_0,WavePlayer_1,WavePlayer_2,WavePlayer_3]
  properties {bShouldExcludeFromBranchCulling=true, weights [1,1,1,1]}
 }
 {id=WavePlayer_0, properties {soundWaveAssetPtr=&/AloeCrouton_Launcher_Prj_Whip_Close_01}}
 {id=WavePlayer_1, properties {soundWaveAssetPtr=&/AloeCrouton_Launcher_Prj_Whip_Close_02}}
 {id=WavePlayer_2, properties {soundWaveAssetPtr=&/AloeCrouton_Launcher_Prj_Whip_Close_03}}
 {id=WavePlayer_3, properties {soundWaveAssetPtr=&/AloeCrouton_Launcher_Prj_Whip_Close_04}}
]
properties {
 attenuationSettings=&/AloeCrouton_Launcher_Prj_Whip_Att
 bOverrideConcurrency=true
 maxDistance=8500
 soundClassObject=&Fort_WeaponRanged_Projectile_Effects
 volumeMultiplier=1.25
 attenuationOverrides {
  bAttenuate=false
  bEnableReverbSend=false
  bSpatialize=false
  bUseComplexCollisionForOcclusion=true
  falloffDistance=1000
  occlusionVolumeAttenuation=0
  stereoSpread=0
  attenuationShapeExtents {x=512}
 }
 concurrencyOverrides {maxCount=3}
}
