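# Sound cue for the distant whip of the AloeCrouton launcher projectile, classed as a ranged weapon projectile effect.
# A Random node picks one of three equally weighted wave variations, and a Modulator above it randomly raises the pitch slightly.
# Spatialization comes from the shared whip attenuation asset, while the cue boosts its volume and overrides concurrency to cap simultaneous instances.
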SoundCue /AloeCroutonWeapon/Sound/AloeCrouton_Launcher_Prj_Whip_Distant_Cue
root=Modulator_0
nodes [
 {id=Modulator_0, children [Random_1], properties {pitchMax=1.2, pitchMin=1}}
 {id=Random_1, children [WavePlayer_0,WavePlayer_1,WavePlayer_2], properties {bShouldExcludeFromBranchCulling=true, weights [1,1,1]}}
 {id=WavePlayer_0, properties {soundWaveAssetPtr=&/AloeCrouton_Launcher_Prj_Whip_Distant_01}}
 {id=WavePlayer_1, properties {soundWaveAssetPtr=&/AloeCrouton_Launcher_Prj_Whip_Distant_02}}
 {id=WavePlayer_2, properties {soundWaveAssetPtr=&/AloeCrouton_Launcher_Prj_Whip_Distant_03}}
]
properties {
 attenuationSettings=&/AloeCrouton_Launcher_Prj_Whip_Att
 bOverrideConcurrency=true
 maxDistance=8500
 soundClassObject=&Fort_WeaponRanged_Projectile_Effects
 volumeMultiplier=1.75
 attenuationOverrides {
  bEnableReverbSend=false
  bUseComplexCollisionForOcclusion=true
  falloffDistance=1000
  occlusionVolumeAttenuation=0
  stereoSpread=0
  attenuationShapeExtents {x=512}
 }
 concurrencyOverrides {maxCount=3}
}
