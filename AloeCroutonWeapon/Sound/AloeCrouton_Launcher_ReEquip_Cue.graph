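# Sound cue for re-equipping the AloeCrouton launcher, built on the launcher's drop sound wave.
# A single wave player feeds a modulator left at default settings, at reduced cue volume.
# It borrows the weapon-drop attenuation, the loot-collect sound class and a pickup concurrency set.
# The concurrency overrides specify one instance with new plays prevented, though the set's name suggests oldest-voice resolution.
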SoundCue /AloeCroutonWeapon/Sound/AloeCrouton_Launcher_ReEquip_Cue
root=Modulator_0
nodes [
 {id=Modulator_0, children [WavePlayer_0]}
 {id=WavePlayer_0, properties {soundWaveAssetPtr=&/AloeCrouton_Launcher_Drop}}
]
properties {
 attenuationSettings=&Fort_Weapon_Drop_Att
 maxDistance=3000
 soundClassObject=&Fort_Loot_Collect
 volumeMultiplier=0.3231
 concurrencyOverrides {maxCount=1, resolutionRule=PreventNew}
 concurrencySet [&PickupMaximum_Oldest]
}
